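# Physics asset for the TangyRadish microphone, a Sparks instrument cosmetic, giving collision bodies and joints to its handle and dynamic vine bones.
# Single capsules cover the handle and each of the three vine segments.
# Constraints chain the segments in sequence from the handle to the third segment, with twist locked on every joint so the vine bends only through swing.
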PhysicsAsset /SparksCosmetics/Instrument/Mic/TangyRadish/Meshes/TangyRadish_Mic_Physics
bodies {
 dyn_vine_01 [{shapeType=Capsule, length=1.233, radius=1.0022, shapeName=dyn_vine_01_capsule, rotation=<rot 0 -90 -90>}]
 dyn_vine_02 [{shapeType=Capsule, length=2.066, radius=1.0757, shapeName=dyn_vine_02_capsule, center=<0.5 0 0>, rotation=<rot 0 -90 -90>}]
 dyn_vine_03 [{shapeType=Capsule, length=2.4268, radius=1.0784, shapeName=dyn_vine_03_capsule, center=<1 0 0>, rotation=<rot 0 -90 -90>}]
 handle [{shapeType=Capsule, length=8.7241, radius=2.0379, center=<0 0 -5.4381>, rotation=<rot 0 -90 0>}]
}
constraints [
 {bone1Name=dyn_vine_03, bone2Name=dyn_vine_02, swing1LimitDegrees=45, swing2LimitDegrees=90, twistLimitDegrees=45, twistMotion=Locked}
 {bone1Name=dyn_vine_01, bone2Name=handle, swing1LimitDegrees=45, swing2LimitDegrees=35, twistLimitDegrees=45, twistMotion=Locked}
 {bone1Name=dyn_vine_02, bone2Name=dyn_vine_01, swing1LimitDegrees=60, swing2LimitDegrees=90, twistLimitDegrees=45, twistMotion=Locked}
]
